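# Additive, two-sided unlit master material for chromatic spark mesh particles, previewed on a firework spark sphere.
# World position offset scales and offsets the mesh, then thickens lines via SplineThicken with screen-size-compensated width.
# Trim texture, particle color and sphere-masked panning chromatic noise form a hue-shifted HDR emissive, with depth fade only at Epic and default quality.
# Dynamic parameters expose hue, panning, offset and width; vertex color supplies per-object randomness.
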Material /DealershipSharedLibrary/FX/Materials/MasterMaterials/Sparks/M_DLR_ChromaticSpark
graph {
 properties {
  bEnableResponsiveAA=true
  bUsedWithMeshParticles=true
  bUsedWithNiagaraMeshParticles=true
  blendMode=BLEND_Additive
  pixelDepthOffsetMode=PDOM_Legacy
  previewMesh=/Game/Missions/Secondary/Events/HonkingFirework/FX/Meshes/SM_Fireworks_Sparks_Sphere_02
  shadingModel=MSM_Unlit
  twoSided=true
  usedShadingModels=Unlit
 }
 nodes [
  n1 = MaterialFunctionCall(Float2=Subtract(A=TextureCoordinate()) {constB=0.5}) {materialFunction=&BreakOutFloat2Components}
  n2 = MaterialFunctionCall(Float2=VectorParameter() {parameterName=XY Scale, defaultValue=<rgba 1 1 0 0>}.RGB) {materialFunction=&BreakOutFloat2Components}
  n3 = AppendVector(A=AppendVector(A=Multiply(A=@n1.R, B=@n2.R), B=Multiply(A=@n1.G, B=@n2.G)), B=Constant() {r=1})
  n4 = Transform(None=VectorParameter() {parameterName=OffsetDistance, defaultValue=<rgba 0 0 1 0>}.RGB)
  n5 = LinearInterpolate(Alpha=NamedRerouteUsage()) {constB=150}
  n6 = Clamp(None=NamedRerouteUsage()) {minDefault=0.5}
  n7 = PixelDepth()
  n8 = ScalarParameter() {defaultValue=0.7, desc=Multiplier for adding spline width based on size on screen, group=SplineWidth, parameterName=ScreenWidthMult}
  n9 = Multiply(A=ComponentMask(None=ViewSize()) {g=true}, B=@n8)
  n10 = Add(A=Clamp(
    None=Divide(A=@n7.R, B=@n9)
    Max=ScalarParameter() {desc=Max clamp on how much screen size can add to spline width, group=SplineWidth, parameterName=ScreenWidthMaxAdd}
   ))
  n11 = Multiply(A=ScalarParameter() {defaultValue=20, group=SplineWidth, parameterName=WidthBase}, B=@n10)
  n12 = Multiply(A=NamedRerouteUsage(), B=Multiply(A=ScalarParameter() {defaultValue=20, group=SplineWidth, parameterName=WidthTip}, B=@n10))
  Add(
   A=Add(A=Transform(None=@n3), B=Multiply(A=@n4, B=@n5))
   B=MaterialFunctionCall(WidthBase=Multiply(A=@n6, B=@n11), WidthTip=Multiply(A=@n6, B=@n12)) {materialFunction=&SplineThicken}.WorldPositionOffset
  )
  n13 = NamedRerouteUsage()
  n14 = NamedRerouteUsage()
  n15 = TextureSample(UVs=TextureCoordinate()) {texture=/Game/Effects/Fort_Effects/Textures/Fire/2DGasTorch_Trim}
  n16 = TextureCoordinate() {uTiling=0}
  n17 = AppendVector(A=Constant(), B=NamedRerouteUsage())
  n18 = ScalarParameter() {defaultValue=1, parameterName=GlowTip_Radius_1}
  n19 = ScalarParameter() {parameterName=GlowTip_Hardness_1}
  n20 = Add(A=Multiply(A=NamedRerouteUsage(), B=NamedRerouteUsage()) {constB=14}, B=TextureCoordinate())
  n21 = ComponentMask(None=@n20) {r=true}
  n22 = ScalarParameter() {defaultValue=0.2, group=WPO, parameterName=Aplha Tile U}
  n23 = ComponentMask(None=@n20) {g=true}
  n24 = ScalarParameter() {defaultValue=0.1, group=WPO, parameterName=Aplha Tile V}
  n25 = Panner(
   Coordinate=AppendVector(A=Multiply(A=@n21, B=@n22), B=Multiply(A=@n23, B=@n24))
   Time=Time() {bOverride_Period=true, period=64}
   Speed=AppendVector(
    A=ScalarParameter() {group=WPO, parameterName=Aplha Speed U}
    B=ScalarParameter() {defaultValue=-0.1, group=WPO, parameterName=Aplha Speed V}
   )
  ) {speedX=-1, speedY=0.1}
  n26 = Multiply(A=@n25, B=Constant2Vector() {g=3, r=3})
  n27 = Multiply(A=Multiply(
    A=TextureSample(UVs=@n25) {texture=/Game/Effects/Fort_Effects/Textures/Tile/T_Noise_02}.RGB
    B=TextureSample(UVs=@n26) {texture=/Game/Effects/Fort_Effects/Textures/Tile/T_Smooth_Tile_Chromatic}.RGB
   )) {constB=32}
  n28 = Multiply(A=Saturate(None=SphereMask(A=@n16, B=@n17, Radius=@n18, Hardness=@n19)), B=Saturate(None=Power(Base=@n27)))
  n29 = Multiply(A=NamedRerouteUsage(), B=Multiply(A=ParticleColor().RGB, B=Multiply(A=@n15.RGB, B=@n28)))
  n30 = Saturate(None=@n29)
  MaterialFunctionCall(Emissive=MaterialFunctionCall(
    Hue Shift Percentage=Multiply(A=@n13, B=@n14)
    Texture=MaterialFunctionCall(LDR Input=@n30, HDR Tint=@n29) {materialFunction=&DeriveHDRfromLDR}.RemappedRange
   ) {materialFunction=&HueShift}.Result) {materialFunction=&MF_FX_EmissiveAdjustments}
  n31 = VertexColor()
  NamedRerouteDeclaration(None=@n31.R) {name=RandomPerObject_1, nodeColor=<rgba 1 0.0156 0.6498 1>}
  NamedRerouteDeclaration(None=@n31.G) {name=RandomPerObject_2, nodeColor=<rgba 1 0.0156 0.6498 1>}
  n32 = DynamicParameter() {defaultValue=<rgba 1 1 -1 1>, paramNames [HueShift,PanSmoke,OffsetMesh,LineWidth]}
  NamedRerouteDeclaration(None=@n32.OffsetMesh) {name=OffsetMesh, nodeColor=<rgba 0.4459 0.0012 1 1>}
  NamedRerouteDeclaration(None=@n32.LineWidth) {name=LineWidth, nodeColor=<rgba 0.4459 0.0012 1 1>}
  NamedRerouteDeclaration(None=@n32.HueShift) {name=HueShift, nodeColor=<rgba 0.4459 0.0012 1 1>}
  NamedRerouteDeclaration(None=@n32.PanSmoke) {name=PanSmoke, nodeColor=<rgba 0.4459 0.0012 1 1>}
  n33 = Multiply(A=Multiply(A=NamedRerouteUsage(), B=ParticleColor().A), B=Add(A=Multiply(A=@n15.RGB) {constB=2}, B=Desaturation(None=@n28)))
  n34 = DepthFade(Opacity=@n33, FadeDistance=ScalarParameter() {defaultValue=100, parameterName=DepthFade})
  QualitySwitch(Default=@n34, Low=@n33, High=@n33, Medium=@n33, Epic=@n34)
 ]
}
